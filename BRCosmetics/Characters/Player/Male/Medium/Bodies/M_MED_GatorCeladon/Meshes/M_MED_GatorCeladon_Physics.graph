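# Physics asset for the GatorCeladon medium male player body cosmetic, covering only the upper torso and a two-bone necklace, apparently for secondary necklace motion.
# Capsules on spine_04, spine_05 and both pec deform bones form the chest volume around the necklace capsules.
# The necklace chains from spine_05 through dyn_necklace_01 to dyn_necklace_02 with restricted swing and no twist range.
# Spine_05 and the pecs attach to spine_04 with free angular motion.
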PhysicsAsset /BRCosmetics/Characters/Player/Male/Medium/Bodies/M_MED_GatorCeladon/Meshes/M_MED_GatorCeladon_Physics
bodies {
 deform_pec_aim_l [{shapeType=Capsule, length=9.131, radius=5.8995, shapeName=deform_pec_aim_l_capsule, center=<7.4036 -6.3695 0>, rotation=<rot 90 33.8897 0>}]
 deform_pec_aim_r [{shapeType=Capsule, length=9.131, radius=5.8995, shapeName=deform_pec_aim_r_capsule, center=<-7.4036 6.3695 0>, rotation=<rot 90 -146.1103 0>}]
 dyn_necklace_01 [{shapeType=Capsule, length=7.0692, radius=1.59, shapeName=dyn_necklace_01_capsule, center=<6.0712 -1.0826 0>, rotation=<rot 0 -90 0>}]
 dyn_necklace_02 [{shapeType=Capsule, length=9.8955, radius=2.7207, shapeName=dyn_necklace_02_capsule, center=<0.2168 1.7185 0>}]
 spine_04 [{shapeType=Capsule, length=16.5, radius=12.058, shapeName=spine_04_capsule, center=<6.2256 -2.344 0>, rotation=<rot 0 -90 0>}]
 spine_05 [{
   shapeType=Capsule
   length=10.1656
   radius=9.2984
   shapeName=spine_05_capsule
   center=<8.4535 -0.1546 0.0082>
   rotation=<rot 0.1274 -90 -0.0195>
  }]
}
constraints [
 {bone1Name=dyn_necklace_01, bone2Name=spine_05, swing1LimitDegrees=45, swing2LimitDegrees=15, twistLimitDegrees=0}
 {bone1Name=dyn_necklace_02, bone2Name=dyn_necklace_01, swing1LimitDegrees=45, swing2LimitDegrees=45, twistLimitDegrees=0}
 {
  bone1Name=spine_05
  bone2Name=spine_04
  swing1LimitDegrees=45
  swing1Motion=Free
  swing2LimitDegrees=45
  swing2Motion=Free
  twistLimitDegrees=45
  twistMotion=Free
 }
 {
  bone1Name=deform_pec_aim_r
  bone2Name=spine_04
  swing1LimitDegrees=45
  swing1Motion=Free
  swing2LimitDegrees=45
  swing2Motion=Free
  twistLimitDegrees=45
  twistMotion=Free
 }
 {
  bone1Name=deform_pec_aim_l
  bone2Name=spine_04
  swing1LimitDegrees=45
  swing1Motion=Free
  swing2LimitDegrees=45
  swing2Motion=Free
  twistLimitDegrees=45
  twistMotion=Free
 }
]
