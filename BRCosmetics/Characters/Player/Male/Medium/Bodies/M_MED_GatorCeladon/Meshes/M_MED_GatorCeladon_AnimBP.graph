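# Animation Blueprint for the GatorCeladon medium male player body cosmetic, adding secondary motion on top of the character's animation.
# It copies the attached parent mesh's pose into the body's Control Rig, which also receives the skydiving flag.
# A RigidBody node then simulates an override physics asset in spine_05 base-bone space with custom gravity and motion limits.
# A final additive roll tweak targets dyn_necklace_01, with an inverted alpha scale-bias.
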AnimBlueprint /BRCosmetics/Characters/Player/Male/Medium/Bodies/M_MED_GatorCeladon/Meshes/M_MED_GatorCeladon_AnimBP
target_skeleton=&/M_MED_GatorCeladon_Skeleton
graphs [
 AnimGraph {nodes [
   n1 = Control Rig!(Source=Copy Pose From Mesh() {bUseAttachedParent=true}, Alpha, IsSkydiving:bool=Get bIsSkyDiving()) -> (Pose) {
    alphaBoolBlend {blendOption=Linear}
    controlRigAssetReference {blueprintRigClass=&/M_MED_GatorCeladon_CtrlRig.M_MED_GatorCeladon_CtrlRig_C}
   }
   n2 = RigidBody!(ComponentPose=Local To Component(LocalPose=@n1), Alpha) -> (Pose) {
    bOverrideWorldGravity=true
    overridePhysicsAsset=&/M_MED_GatorCeladon_Physics
    simulationSpace=BaseBoneSpace
    alphaBoolBlend {blendOption=Linear}
    baseBoneRef {boneName=spine_05}
    componentAppliedLinearAccClamp=<2500 2500 2500>
    overrideWorldGravity {z=-980}
    simSpaceSettings {
     maxAngularAcceleration=65
     maxAngularVelocity=1500
     maxLinearAcceleration=1000
     maxLinearVelocity=3000
     velocityScaleZ=0.3
     worldAlpha=1
     externalLinearDragV=<1 1 1>
    }
   }
   n3 = "Transform (Modify) Bone - Bone: dyn_necklace_01"!(ComponentPose=@n2, Alpha=1, Rotation="0,0,-6") -> (Pose) {
    rotationMode=BMM_Additive
    alphaBoolBlend {blendOption=Linear}
    alphaScaleBias {bias=1, scale=-1}
    boneToModify {boneName=dyn_necklace_01}
    rotation {roll=-6}
   }
   Output Pose(Result=Component To Local(ComponentPose=@n3)) {layerGroup=None}
  ]}
 EventGraph {nodes [Event BlueprintUpdateAnimation()]}
]
variables []
