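# Control Rig for the face accessory part of the GatorCeladon medium male character, correcting the hair_base bone under the head.
# On execution, an aim constraint with local-space offset orients hair_base using neck_01 as its target.
# A plane clamp of hair_base in neck_01 space feeds a Set Translation, but that node is outside the execution chain.
# An unexecuted deform_strap_b transform tweak and a loose integer select also remain.
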ControlRigBlueprint /BRCosmetics/Characters/Player/Male/Medium/Bodies/M_MED_GatorCeladon/Meshes/Parts/M_MED_GatorCeladon_FaceAcc_CtrlRig
graphs [
 RigVMFunctionLibrary {links [], nodes []}
 RigVMModel {nodes [
   RigUnit_BeginExecution = &RigUnit_BeginExecution()
   &RigUnit_AimConstraintLocalSpaceOffset(
    AdvancedSettings=(DebugSettings=(bEnabled=true,Scale=10,WorldOffset=(Rotation=(X=0,Y=0,Z=0,W=1),Translation=(X=0,Y=0,Z=0),Scale3D=(X=1,Y=1,Z=1))),RotationOrderForFilter=XZY)
    Child=(Type=Bone,Name="hair_base")
    Filter=(bX=true,bY=true,bZ=True)
    Parents=((Item=(Type=Bone,Name="neck_01"),Weight=1))
    UpAxis=(X=0,Y=-1,Z=0)
    Weight=1
    WorldUp=(Target=(X=0,Y=0,Z=1),Kind=Location,Space=(Type=Bone,Name="hair_base"))
    ExecutePin=@RigUnit_BeginExecution.ExecutePin
   )
   &RigUnit_ModifyTransforms(
    ExecutePin=()
    ItemToModify=((Item=(Type=Bone,Name="deform_strap_b"),Transform=(Rotation=(X=0,Y=0,Z=0.0872,W=0.9962),Translation=(X=0,Y=0,Z=0),Scale3D=(X=1,Y=1,Z=1))))
    Weight=1
    WeightMinimum=0
   )
   RigVMFunction_MathBoolToInteger = &RigVMFunction_MathBoolToInteger(Value=False)
   RigVMDispatchNode(Values=(4,4), Index=@RigVMFunction_MathBoolToInteger.Result) {template="DISPATCH_RigVMDispatch_SelectInt32(in Index,in Values,out Result)"}
   n1 = &RigUnit_GetTransform(Item=(Type=Bone,Name="neck_01"), bInitial=False)
   ClampSpatially = /Script/RigVM.RigVMFunction_MathTransformClampSpatially(
    Axis=X
    DebugColor=(R=1,G=0,B=0,A=1)
    DebugThickness=1
    Maximum=25
    Minimum=10
    Space=(Rotation=(X=0,Y=0,Z=0,W=1),Translation=(X=0,Y=0,Z=0),Scale3D=(X=1,Y=1,Z=1))
    Type=Plane
    bDrawDebug=true
    Value=&RigUnit_GetTransform(Item=(Type=Bone,Name="hair_base"), bInitial=False).Transform
    Space.Rotation=@n1.Transform.Rotation
    Space.Translation=@n1.Transform.Translation
   ) {
    is=struct RigVMUnitNode
    template="ClampSpatially::Execute(in Value,in Axis,in Type,in Minimum,in Maximum,in Space,in bDrawDebug,in DebugColor,in DebugThickness,out Result)"
   }
   Set Transform_8 = /Script/ControlRig.RigUnit_SetTranslation(
    ExecutePin=()
    Item=(Type=Bone,Name="hair_base")
    Space=GlobalSpace
    Weight=1
    bInitial=false
    bPropagateToChildren=true
    Value=@ClampSpatially.Result.Translation
   ) {is=struct RigVMUnitNode, template="Set Transform::Execute(in Item,in Space,in bInitial,in Value,in Weight,in bPropagateToChildren)"}
  ]}
]
hierarchy [
 root {loc=<0 0 0>, rot=<0 0 0 1>}
 pelvis {parent=root, loc=<0 -0.7071 93.1191>, rot=<0.0127 0.707 -0.0127 -0.707>}
 spine_01 {parent=pelvis, loc=<6.7494 -0.6388 0>, rot=<0 0 0.034 -0.9994>}
 spine_02 {parent=spine_01, loc=<7.2518 -0.035 0>, rot=<0 0 0.0238 -0.9997>}
 spine_03 {parent=spine_02, loc=<6.0104 0 0>, rot=<0 0 0.0149 -0.9999>}
 spine_04 {parent=spine_03, loc=<8.1024 0 0>, rot=<0 0 -0.0132 -0.9999>}
 spine_05 {parent=spine_04, loc=<10.0146 0 0>, rot=<0 0 -0.0015 -1>}
 neck_01 {parent=spine_05, loc=<11.3138 0.1529 0>, rot=<0 0 -0.2003 -0.9797>}
 neck_02 {parent=neck_01, loc=<4.5627 0 0>, rot=<0 0 0.0504 -0.9987>}
 head {parent=neck_02, loc=<5.3914 0 0>, rot=<0 0 0.115 -0.9934>}
 dyn_hair_l {parent=head, loc=<7.8749 1.3044 -8.9979>, rot=<0.0757 0.6971 0.7111 -0.0508>}
 dyn_up_hair_01 {parent=head, loc=<16.4295 -6.977 -3.9348>, rot=<0.4138 -0.0648 0.0811 -0.9044>}
 dyn_up_hair_02 {parent=head, loc=<15.8375 -7.4336 -4.3951>, rot=<0.374 -0.4008 0.2187 -0.8073>}
 hair_base {parent=head, loc=<4.2946 -2.0777 0>, rot=<0 0 0 1>}
 dyn_mid_hair_01 {parent=hair_base, loc=<-3.2026 -4.0764 0.074>, rot=<0 -1 0 0>}
 dyn_left_back_hair {parent=hair_base, loc=<-2.1225 -1.7553 -4.4703>, rot=<0.0175 -0.9338 -0.3524 -0.059>}
 dyn_reft_back_hair {parent=hair_base, loc=<-2.1224 -1.7553 4.4703>, rot=<0.0212 -0.9095 0.4112 0.0578>}
 dyn_hair_r {parent=head, loc=<7.8753 1.3044 8.9979>, rot=<-0.0757 -0.6971 0.7111 -0.0508>}
 dyn_hair_f_l {parent=head, loc=<12.7644 8.8992 -6.6038>, rot=<-0.0063 0.4037 0.9149 -0.0005>}
 dyn_hair_up_base {parent=head, loc=<15.753 -1.7292 0>, rot=<0 0 0.1621 -0.9868>}
 dyn_hair_f_r {parent=head, loc=<12.8 8.419 7.2487>, rot=<-0.0261 -0.4703 0.882 0.0121>}
 dyn_hair_f {parent=head, loc=<13.4936 11.909 0>, rot=<0 0 0.9995 0.0306>}
]
