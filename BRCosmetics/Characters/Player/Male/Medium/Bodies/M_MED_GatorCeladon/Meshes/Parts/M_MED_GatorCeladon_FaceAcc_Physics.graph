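# Physics asset for the face accessory part of the male medium GatorCeladon player cosmetic, defining bodies and joints for its hair physics.
# Mostly capsule bodies cover the front, side, back and upper hair bones, plus head, neck and upper spine.
# Back strands hang off a sphere hair base with a tight joint to the head; other strands constrain directly to the head with narrow angular limits.
# The upper-hair base is angularly locked, while head-neck and neck-spine joints rotate freely.
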PhysicsAsset /BRCosmetics/Characters/Player/Male/Medium/Bodies/M_MED_GatorCeladon/Meshes/Parts/M_MED_GatorCeladon_FaceAcc_Physics
bodies {
 dyn_hair_f [{shapeType=Capsule, length=3, radius=2, shapeName=dyn_hair_f_capsule, center=<1.3107 -1.947 0>}]
 dyn_hair_f_l [{shapeType=Capsule, length=3, radius=2, shapeName=dyn_hair_f_l_capsule, center=<1.3107 -1.947 0>}]
 dyn_hair_f_r [{shapeType=Capsule, length=3, radius=2, shapeName=dyn_hair_f_r_capsule, center=<1.3107 -1.947 0>}]
 dyn_hair_l [{shapeType=Capsule, length=5.3273, radius=2, shapeName=dyn_hair_l_capsule, center=<4.5646 -1.2398 0>, rotation=<rot 90 0 0>}]
 dyn_hair_r [{shapeType=Capsule, length=5.3273, radius=2, shapeName=dyn_hair_r_capsule, center=<4.5646 -1.2398 0>, rotation=<rot 90 0 0>}]
 dyn_hair_up_base [{shapeType=Capsule, length=2.3375, radius=3.2568, shapeName=dyn_hair_up_base_capsule}]
 dyn_left_back_hair [{shapeType=Capsule, length=3, radius=2, shapeName=dyn_left_back_hair_capsule, center=<2.9886 -1.1546 -0.1188>, rotation=<rot 20.7729 0 0>}]
 dyn_mid_hair_01 [{shapeType=Capsule, length=5.2007, radius=2, shapeName=dyn_mid_hair_01_capsule, center=<2.587 -1.1546 0>}]
 dyn_reft_back_hair [{shapeType=Capsule, length=3, radius=2, shapeName=dyn_reft_back_hair_capsule, center=<2.587 -1.1546 0>, rotation=<rot -10.0432 0 0>}]
 dyn_up_hair_01 [{shapeType=Capsule, length=4.1629, radius=1.4643, shapeName=dyn_up_hair_01_capsule, center=<3.2794 0 0>, rotation=<rot 90 0 0>}]
 dyn_up_hair_02 [{shapeType=Capsule, length=4.1629, radius=1.4643, shapeName=dyn_up_hair_02_capsule, center=<3.2794 0 0>, rotation=<rot 90 0 0>}]
 hair_base [{shapeType=Sphere, radius=2, shapeName=hair_base_sphere, center=<2 0 0>}]
 head [{shapeType=Capsule, length=0, radius=1.5, center=<10.4832 -0.1625 0.0104>, rotation=<rot 0 -50.3416 -88.7933>}]
 neck_01 [{shapeType=Capsule, length=7.4257, radius=5.6502, shapeName=neck_01_capsule, center=<0 1.8213 0>, rotation=<rot 90 0 0>}]
 spine_05 [{shapeType=Capsule, length=11.1269, radius=11.2019, shapeName=spine_05_capsule, center=<0.3167 0.5394 0>, rotation=<rot 0 -90 0>}]
}
constraints [
 {bone1Name=hair_base, bone2Name=head, swing1LimitDegrees=10, swing2LimitDegrees=10, twistLimitDegrees=1}
 {bone1Name=dyn_left_back_hair, bone2Name=hair_base, swing1LimitDegrees=20, swing2LimitDegrees=15, twistLimitDegrees=5}
 {bone1Name=dyn_mid_hair_01, bone2Name=hair_base, swing1LimitDegrees=25, swing2LimitDegrees=15, twistLimitDegrees=5}
 {bone1Name=dyn_hair_f, bone2Name=head, swing1LimitDegrees=30, swing2LimitDegrees=15, twistLimitDegrees=5}
 {bone1Name=dyn_hair_f_l, bone2Name=head, swing1LimitDegrees=30, swing2LimitDegrees=15, twistLimitDegrees=5}
 {bone1Name=dyn_hair_f_r, bone2Name=head, swing1LimitDegrees=30, swing2LimitDegrees=15, twistLimitDegrees=5}
 {bone1Name=dyn_hair_l, bone2Name=head, swing1LimitDegrees=20, swing2LimitDegrees=25, twistLimitDegrees=5}
 {bone1Name=dyn_hair_r, bone2Name=head, swing1LimitDegrees=20, swing2LimitDegrees=25, twistLimitDegrees=5}
 {
  bone1Name=dyn_hair_up_base
  bone2Name=head
  swing1LimitDegrees=45
  swing1Motion=Locked
  swing2LimitDegrees=45
  swing2Motion=Locked
  twistLimitDegrees=45
  twistMotion=Locked
 }
 {bone1Name=dyn_up_hair_01, bone2Name=head, swing1LimitDegrees=30, swing2LimitDegrees=30, twistLimitDegrees=10}
 {bone1Name=dyn_up_hair_02, bone2Name=head, swing1LimitDegrees=30, swing2LimitDegrees=30, twistLimitDegrees=10}
 {bone1Name=dyn_reft_back_hair, bone2Name=hair_base, swing1LimitDegrees=20, swing2LimitDegrees=15, twistLimitDegrees=5}
 {
  bone1Name=head
  bone2Name=neck_01
  swing1LimitDegrees=45
  swing1Motion=Free
  swing2LimitDegrees=45
  swing2Motion=Free
  twistLimitDegrees=45
  twistMotion=Free
 }
 {
  bone1Name=neck_01
  bone2Name=spine_05
  swing1LimitDegrees=45
  swing1Motion=Free
  swing2LimitDegrees=45
  swing2Motion=Free
  twistLimitDegrees=45
  twistMotion=Free
 }
]
