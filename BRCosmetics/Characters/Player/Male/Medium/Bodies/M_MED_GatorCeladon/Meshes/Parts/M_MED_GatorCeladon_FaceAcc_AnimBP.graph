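# Post-process style AnimBP for the GatorCeladon face accessory, animating its hair on the attached character.
# The parent mesh pose feeds a Control Rig, then two RigidBody sims (base and cape physics assets) in spine_05 space, then additive scaling on three hair bones.
# Events compute head yaw and detect the cape backpack.
# The head-driven hair_base transform is unlinked from the output, and the cape flag has no visible consumer.
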AnimBlueprint /BRCosmetics/Characters/Player/Male/Medium/Bodies/M_MED_GatorCeladon/Meshes/Parts/M_MED_GatorCeladon_FaceAcc_AnimBP
target_skeleton=&/M_MED_GatorCeladon_FaceAcc_Skeleton
graphs [
 AnimGraph {nodes [
   "Transform (Modify) Bone - Bone: hair_base"(Alpha=Get Head Rot() -> (Head Rot:double), Translation="0,0,0", Rotation="0,10,0", Scale="1,1,1") {
    rotationMode=BMM_Additive
    rotationSpace=BCS_BoneSpace
    alphaBoolBlend {blendOption=Linear}
    boneToModify {boneName=hair_base}
    rotation {yaw=10}
   }
   n1 = Control Rig(Source=Copy Pose From Mesh() {bUseAttachedParent=true}, Alpha=0.5) {
    alphaBoolBlend {blendOption=Linear}
    alphaScaleBiasClamp {bMapRange=true, outRange {max=0.5, min=1}}
    controlRigAssetReference {blueprintRigClass=&/M_MED_GatorCeladon_FaceAcc_CtrlRig.M_MED_GatorCeladon_FaceAcc_CtrlRig_C}
   }
   n2 = RigidBody!(ComponentPose=Local To Component(LocalPose=@n1), Alpha) -> (Pose) {
    bOverrideWorldGravity=true
    overridePhysicsAsset=&/M_MED_GatorCeladon_FaceAcc_Physics
    simulationSpace=BaseBoneSpace
    alphaBoolBlend {blendOption=Linear}
    alphaScaleBiasClamp {bMapRange=true, outRange {max=0, min=1}}
    baseBoneRef {boneName=spine_05}
    componentAppliedLinearAccClamp=<2500 2500 2500>
    overrideWorldGravity {z=-980}
    simSpaceSettings {maxAngularAcceleration=65, maxAngularVelocity=15, maxLinearAcceleration=1000, maxLinearVelocity=3, velocityScaleZ=0.1, worldAlpha=0.3}
   }
   n3 = RigidBody!(ComponentPose=@n2, bAlphaBoolEnabled:bool) -> (Pose) {
    alphaInputType=Bool
    bOverrideWorldGravity=true
    overridePhysicsAsset=&/M_MED_GatorCeladon_Cape_FaceAcc_Physics1
    simulationSpace=BaseBoneSpace
    alphaBoolBlend {blendOption=Linear}
    baseBoneRef {boneName=spine_05}
    componentAppliedLinearAccClamp=<2500 2500 2500>
    overrideWorldGravity {z=-980}
    simSpaceSettings {maxAngularAcceleration=65, maxAngularVelocity=15, maxLinearAcceleration=1000, maxLinearVelocity=3, velocityScaleZ=0.1, worldAlpha=0.3}
   }
   n4 = "Transform (Modify) Bone - Bone: dyn_mid_hair_01"!(ComponentPose=@n3, bAlphaBoolEnabled:bool=True, Scale="0.5,0.5,1") -> (Pose) {
    alphaInputType=Bool
    rotationSpace=BCS_BoneSpace
    scaleMode=BMM_Additive
    alphaBoolBlend {blendOption=Linear}
    boneToModify {boneName=dyn_mid_hair_01}
    scale=<0.5 0.5>
   }
   n5 = "Transform (Modify) Bone - Bone: dyn_left_back_hair"!(ComponentPose=@n4, bAlphaBoolEnabled:bool=True, Scale="0.5,0.5,1") -> (Pose) {
    alphaInputType=Bool
    rotationSpace=BCS_BoneSpace
    scaleMode=BMM_Additive
    alphaBoolBlend {blendOption=Linear}
    boneToModify {boneName=dyn_left_back_hair}
    scale=<0.5 0.5>
   }
   n6 = "Transform (Modify) Bone - Bone: dyn_reft_back_hair"!(ComponentPose=@n5, bAlphaBoolEnabled:bool=True, Scale="0.5,0.5,1") -> (Pose) {
    alphaInputType=Bool
    rotationSpace=BCS_BoneSpace
    scaleMode=BMM_Additive
    alphaBoolBlend {blendOption=Linear}
    boneToModify {boneName=dyn_reft_back_hair}
    scale=<0.5 0.5>
   }
   Output Pose(Result=Component To Local(ComponentPose=@n6)) {layerGroup=None}
  ]}
 EventGraph {nodes [
   Event BlueprintUpdateAnimation()
   n1 = GetOwnerSkeletalMeshComponent()
   > Set Head Rot(Head Rot:double=MapRangeClamped(
     Value=GetDeltaTransformFromRefPose!(self=@n1, BoneName=head, BaseName=spine_05) -> (
      ReturnValue_Location:Vector
      ReturnValue_Rotation_Roll:float
      ReturnValue_Rotation_Pitch:float
      ReturnValue_Rotation_Yaw:float
      ReturnValue_Scale:Vector
     ).ReturnValue_Rotation_Yaw
     InRangeA=0
     InRangeB=-90
    )) -> (Output_Get:double)
   n2 = Event OnCharacterCustomizationLoaded()
   > Set isWearingCape(isWearingCape="Equal (Object)"(
     A=Get Array Item with EFort Custom Part Type(Array:Array of Custom Character Part Object References=@n2.CustomParts, PartType=Backpack) -> (Output:Custom Character Part Object Reference)
     B=/BRCosmetics/Characters/CharacterParts/Backpacks/CP_Backpack_GatorCeladonCape
    ))
   Get isWearingCape() -> (isWearingCape:bool)
  ]}
]
variables [
 Head Rot {type=number}
 isWearingCape {type=boolean}
]
